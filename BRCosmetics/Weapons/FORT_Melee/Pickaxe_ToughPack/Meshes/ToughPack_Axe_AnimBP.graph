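# Animation Blueprint for the ToughPack pickaxe that adds physical secondary motion to the ring on the axe.
# The AnimGraph passes the input pose through two additive pitch offsets on base_ring, then an Anim Dynamics node simulates dyn_ring_01 with an angular spring, angular limits and reduced gravity.
# Each update, the EventGraph computes a gravity override through a deprecated helper and sets ring simulation-space and alpha variables according to whether it is in the front end.
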AnimBlueprint /BRCosmetics/Weapons/FORT_Melee/Pickaxe_ToughPack/Meshes/ToughPack_Axe_AnimBP
target_skeleton=&/ToughPack_Axe_Skeleton
graphs [
 AnimGraph {nodes [
   AnimGraphNode_SubInput_0 = Input Pose()
   n1 = "Transform (Modify) Bone - Bone: base_ring"(ComponentPose=Local To Component(LocalPose=@AnimGraphNode_SubInput_0), Alpha=1, Translation="0,0,0", Rotation="30,0,0", Scale="1,1,1") {
    rotationMode=BMM_Additive
    rotationSpace=BCS_BoneSpace
    alphaBoolBlend {blendOption=Linear}
    boneToModify {boneName=base_ring}
    rotation {pitch=30}
   }
   n2 = "Transform (Modify) Bone - Bone: base_ring"!(ComponentPose=@n1, bAlphaBoolEnabled:bool=True, Translation="0,0,0", Rotation="-10,0,0", Scale="1,1,1") -> (Pose) {
    alphaInputType=Bool
    rotationMode=BMM_Additive
    rotationSpace=BCS_BoneSpace
    alphaBoolBlend {blendInTime=0.25, blendOption=Linear, blendOutTime=0.25}
    boneToModify {boneName=base_ring}
    rotation {pitch=-10}
   }
   n3 = Anim Dynamics - Bone: dyn_ring_01!(SimulationSpace:AnimPhysSimSpaceType Enum, ComponentPose=@n2, Alpha=1, GravityOverride:Vector="0,0,-980") -> (Pose) {
    angularDampingOverride=1
    angularSpringConstant=7000
    bAngularSpring=true
    bOverrideAngularDamping=true
    bUseGravityOverride=true
    gravityScale=0.35
    simulationSpace=RootRelative
    alphaBoolBlend {blendOption=Linear}
    boundBone {boneName=dyn_ring_01}
    componentAppliedLinearAccClamp=<10000 10000 10000>
    componentLinearAccScale=<0.65 0.65 0.65>
    componentLinearVelScale=<0.65 0.65 0.65>
    gravityOverride {z=-980}
    physicsBodyDefinitions [{
      boundBone {boneName=dyn_ring_01}
      boxExtents=<12 30 12>
      constraintSetup {angularLimitsMax=<0 40 3>, angularLimitsMin=<0 -50 -3>, angularTarget=<1 0 0>}
      localJointOffset=<5 0 0>
     }]
   }
   Output Pose(Result=Component To Local(ComponentPose=@n3)) {layerGroup=None}
  ]}
 EventGraph {nodes [
   Event BlueprintUpdateAnimation()
   > n1 = Gravity Override State Function DEPRECATED(CharacterPartAnimInstance=Self-Reference(), GameTimeInSeconds=GetGameTimeInSeconds(), GravityOverrideParams=Get Gravity Override Params())
   > Set GravityOverride(GravityOverride=@n1)
   > n2 = Branch(Condition=Get bIsInFrontEnd())
   > Set Sim Space Ring(Sim Space Ring:AnimPhysSimSpaceType Enum=World) -> (Output_Get:AnimPhysSimSpaceType Enum)
   > Set ringDynAlpha(ringDynAlpha:double=0.6) -> (Output_Get:double)
   Set Sim Space Ring(execute=@n2.else, Sim Space Ring=BoneRelative)
   > Set ringDynAlpha(ringDynAlpha=1)
   TryGetPawnOwner()
  ]}
]
variables [
 Gravity Override Params {type=GravityOverrideParamsStruct}
 GravityOverride {type=Vector}
 ringDynAlpha {type=number}
 Sim Space Ring {type=AnimPhysSimSpaceType}
]
